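# Physics asset for the dynamic front collar (choker) bones on the female medium Black Widow Jacket body.
# A small master body hangs off spine_05 and carries the collar top, a compound of capsules and angled boxes.
# Three capsule surface bodies (top, mid, bottom) each constrain directly to spine_05, which has its own large capsule.
# All constraints leave swing and twist free, so their stored angle limits do not restrict rotation.
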PhysicsAsset /BRCosmetics/Characters/Player/Female/Medium/Bodies/F_MED_Black_Widow_Jacket/Meshes/F_MED_Black_Widow_Jacket_Dyn_choker
bodies {
 dyn_collar_fr_MSTR [{shapeType=Capsule, length=0.5, radius=0.5}]
 dyn_collar_fr_surface_bott [{shapeType=Capsule, length=4.1875, radius=3.8751, center=<-0.1743 -3.8584 0>, rotation=<rot 90 -90 -87.9926>}]
 dyn_collar_fr_surface_mid [{shapeType=Capsule, length=2.9375, radius=3.8751, center=<-0.0003 -3.9353 0>, rotation=<rot 90 0 1.2335>}]
 dyn_collar_fr_surface_top [{shapeType=Capsule, length=2.9375, radius=3.8751, center=<0 -3.9237 0>, rotation=<rot 90 0 0>}]
 dyn_collar_fr_top [
  {shapeType=Capsule, length=5.2954, radius=1.2665, center=<-1.2908 1.2696 0>, rotation=<rot 0 -6.6796 0>}
  {shapeType=Capsule, length=5.2954, radius=1.2665, center=<1.0058 1.2353 0>}
  {shapeType=Box, extentX=2.2898, extentY=0.9731, extentZ=7.5, center=<-0.2302 0.4231 0>, rotation=<rot 0 1.2333 0>}
  {shapeType=Box, extentX=2.1899, extentY=1.0981, extentZ=7.5, center=<0.06 -0.6626 3.1262>, rotation=<rot 0 15.6634 -45.0019>}
  {shapeType=Box, extentX=2.1899, extentY=1.0981, extentZ=7.5, center=<0.0482 -1.1711 -3.7856>, rotation=<rot 0 15.6673 45.0005>}
 ]
 spine_05 [{shapeType=Capsule, length=12.8031, radius=8.6478, center=<2.0911 -1.9204 -0.7007>, rotation=<rot -84.9975 -0.0001 -34.878>}]
}
constraints [
 {
  bone1Name=dyn_collar_fr_MSTR
  bone2Name=spine_05
  swing1LimitDegrees=45
  swing1Motion=Free
  swing2LimitDegrees=45
  swing2Motion=Free
  twistLimitDegrees=45
  twistMotion=Free
 }
 {
  bone1Name=dyn_collar_fr_top
  bone2Name=dyn_collar_fr_MSTR
  swing1LimitDegrees=45
  swing1Motion=Free
  swing2LimitDegrees=45
  swing2Motion=Free
  twistLimitDegrees=45
  twistMotion=Free
 }
 {
  bone1Name=dyn_collar_fr_surface_top
  bone2Name=spine_05
  swing1LimitDegrees=45
  swing1Motion=Free
  swing2LimitDegrees=45
  swing2Motion=Free
  twistLimitDegrees=45
  twistMotion=Free
 }
 {
  bone1Name=dyn_collar_fr_surface_mid
  bone2Name=spine_05
  swing1LimitDegrees=45
  swing1Motion=Free
  swing2LimitDegrees=45
  swing2Motion=Free
  twistLimitDegrees=45
  twistMotion=Free
 }
 {
  bone1Name=dyn_collar_fr_surface_bott
  bone2Name=spine_05
  swing1LimitDegrees=45
  swing1Motion=Free
  swing2LimitDegrees=45
  swing2Motion=Free
  twistLimitDegrees=45
  twistMotion=Free
 }
]
