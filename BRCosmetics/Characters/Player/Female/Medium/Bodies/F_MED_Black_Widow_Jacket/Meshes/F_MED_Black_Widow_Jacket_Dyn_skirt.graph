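# Physics asset for the female medium Black Widow Jacket body, covering its dynamic skirt panels, right-side belt chain and collar tips.
# Capsules on pelvis, thighs and calves provide body shapes alongside mostly capsule-based cloth bones.
# Tiny MSTR anchor bodies link to the pelvis through free-angular constraints, and chains hang from them with tight angular limits.
# Collars swing on one axis, and a free constraint ties left side and front skirt ends.
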PhysicsAsset /BRCosmetics/Characters/Player/Female/Medium/Bodies/F_MED_Black_Widow_Jacket/Meshes/F_MED_Black_Widow_Jacket_Dyn_skirt
bodies {
 calf_l [{shapeType=Capsule, length=33.6396, radius=5.5589, center=<-19.5026 -2.9284 -0.0004>, rotation=<rot 90 0 -5.1755>}]
 calf_r [{shapeType=Capsule, length=33.6396, radius=5.5589, center=<19.5026 2.9284 0.0004>, rotation=<rot 90 0 -5.1755>}]
 dyn_belt_01_r [{shapeType=Box, extentX=15.1875, extentY=3.3125, extentZ=1.875, center=<8.5434 0.0843 0.01>, rotation=<rot -3.2469 -0.0957 34.3427>}]
 dyn_belt_01_r_MSTR [{shapeType=Capsule, length=0.505, radius=0.505, rotation=<rot 0 -90 0>}]
 dyn_belt_02_r [{
   shapeType=Capsule
   length=8.25
   radius=2.3
   shapeName=dyn_belt_02_r_capsule
   center=<4.9337 -0.2 0.45>
   rotation=<rot 41.9942 92.1314 91.4268>
  }]
 dyn_belt_03_r [{shapeType=Capsule, length=8, radius=2.3, shapeName=dyn_belt_03_r_capsule, center=<4.8265 -0.75 0.55>, rotation=<rot 53.01 89.5255 89.6229>}]
 dyn_belt_04_r [{shapeType=Capsule, length=7, radius=2.3, shapeName=dyn_belt_04_r_capsule, center=<5.2115 -0.55 0.3>, rotation=<rot -47.8481 -91.3856 92.297>}]
 dyn_belt_05_r [{
   shapeType=Capsule
   length=7.5
   radius=2.25
   shapeName=dyn_belt_05_r_capsule
   center=<7.2032 -0.6 -0.2>
   rotation=<rot -30.4984 92.6426 -95.2772>
  }]
 dyn_beltstrap_MSTR [{shapeType=Capsule, length=0.505, radius=0.505, rotation=<rot 0 -90 0>}]
 dyn_collar_01_l [{shapeType=Capsule, length=0, radius=3, shapeName=dyn_collar_01_l_capsule, center=<2.715 0 1.1843>}]
 dyn_collar_01_l_MSTR [{shapeType=Capsule, length=0.505, radius=0.505, rotation=<rot 0 -90 0>}]
 dyn_collar_01_r [{shapeType=Capsule, length=0, radius=3, shapeName=dyn_collar_01_r_capsule, center=<2.715 0 1.1843>}]
 dyn_collar_01_r_MSTR [{shapeType=Capsule, length=0.505, radius=0.505, rotation=<rot 0 -90 0>}]
 dyn_skirt_01_l [{shapeType=Capsule, length=13.25, radius=5.5, shapeName=dyn_skirt_01_l_capsule, center=<5.2199 -0.4 4>, rotation=<rot 0 0.0003 -97.1692>}]
 dyn_skirt_01_l_MSTR [{shapeType=Capsule, length=0.505, radius=0.505, rotation=<rot 0 -90 0>}]
 dyn_skirt_01_r [{
   shapeType=Capsule
   length=12.5
   radius=5
   shapeName=dyn_skirt_01_r_capsule
   center=<6.3062 -2.3829 3.9753>
   rotation=<rot -4.3866 11.1547 -89.5787>
  }]
 dyn_skirt_01_r_MSTR [{shapeType=Capsule, length=0.505, radius=0.505, rotation=<rot 0 -90 0>}]
 dyn_skirt_02_l [{shapeType=Capsule, length=13.25, radius=5.5, shapeName=dyn_skirt_02_l_capsule, center=<6.581 -0.4 4>, rotation=<rot 0 0.0003 -97.1692>}]
 dyn_skirt_02_r [{
   shapeType=Capsule
   length=12.5
   radius=5
   shapeName=dyn_skirt_02_r_capsule
   center=<6.5196 -2.2463 2.1255>
   rotation=<rot -15.3777 10.6241 -89.5637>
  }]
 dyn_skirt_bk_01_l [{shapeType=Capsule, length=10, radius=5, shapeName=dyn_skirt_bk_01_l_capsule, center=<5.1042 0 1.8>, rotation=<rot -0.0001 0 -90.2763>}]
 dyn_skirt_bk_01_l_MSTR [{shapeType=Capsule, length=0.505, radius=0.505, rotation=<rot 0 -90 0>}]
 dyn_skirt_bk_01_r [{shapeType=Capsule, length=10, radius=5, shapeName=dyn_skirt_bk_01_r_capsule, center=<5.1042 0 1.8>, rotation=<rot -0.0001 0 -90.2763>}]
 dyn_skirt_bk_01_r_MSTR [{shapeType=Capsule, length=0.505, radius=0.505, rotation=<rot 0 -90 0>}]
 dyn_skirt_fr_01_l [{
   shapeType=Capsule
   length=8.75
   radius=6.5
   shapeName=dyn_skirt_fr_01_l_capsule
   center=<6.7524 -3.0563 4.4962>
   rotation=<rot 0.0004 0.0001 -89.0386>
  }]
 dyn_skirt_fr_01_l_MSTR [{shapeType=Capsule, length=0.505, radius=0.505, rotation=<rot 0 -90 0>}]
 dyn_skirt_fr_01_r [{
   shapeType=Capsule
   length=15
   radius=5.5
   shapeName=dyn_skirt_fr_01_r_capsule
   center=<7.0923 2.4 3.8>
   rotation=<rot -2.2621 14.1615 109.5732>
  }]
 dyn_skirt_fr_01_r_MSTR [{shapeType=Capsule, length=0.505, radius=0.505, rotation=<rot 0 -90 0>}]
 dyn_skirt_fr_02_l [
  {
   shapeType=Capsule
   length=8.75
   radius=7
   shapeName=dyn_skirt_fr_02_l_capsule
   center=<9.6832 -3.2421 4.2157>
   rotation=<rot -0.6574 -0.7331 -90.2505>
  }
  {
   shapeType=Capsule
   length=7.75
   radius=7
   shapeName=dyn_skirt_fr_02_l_capsule
   center=<5.9154 -12.2745 -1.4034>
   rotation=<rot -10.4672 -20.135 40.0488>
  }
 ]
 dyn_skirt_fr_02_r [{
   shapeType=Capsule
   length=6.4738
   radius=5.5
   shapeName=dyn_skirt_fr_02_r_capsule
   center=<6.426 5.4544 2.3864>
   rotation=<rot -2.6683 16.7769 109.4607>
  }]
 pelvis [
  {shapeType=Capsule, length=9.9209, radius=8.5589, center=<0 -1.4951 -0.0001>}
  {shapeType=Capsule, length=11.7959, radius=10.2464, center=<-6.0656 -1.5683 0>}
 ]
 thigh_l [{shapeType=Capsule, length=24.4501, radius=9.5, center=<-19.5967 0.0346 1.7071>, rotation=<rot 89.4052 180 -179.9999>}]
 thigh_r [{shapeType=Capsule, length=24.7052, radius=9.5, center=<19.5467 0 -1.7>, rotation=<rot 89.4 180 179>}]
}
constraints [
 {
  bone1Name=dyn_collar_01_l
  bone2Name=dyn_collar_01_l_MSTR
  swing1LimitDegrees=45
  swing1Motion=Locked
  swing2LimitDegrees=10
  twistLimitDegrees=5
  twistMotion=Locked
 }
 {
  bone1Name=dyn_collar_01_r
  bone2Name=dyn_collar_01_r_MSTR
  swing1LimitDegrees=45
  swing1Motion=Locked
  swing2LimitDegrees=10
  twistLimitDegrees=5
  twistMotion=Locked
 }
 {bone1Name=dyn_skirt_fr_01_r, bone2Name=dyn_skirt_fr_01_r_MSTR, swing1LimitDegrees=5, swing2LimitDegrees=30, twistLimitDegrees=5}
 {bone1Name=dyn_belt_01_r, bone2Name=dyn_belt_01_r_MSTR, swing1LimitDegrees=18, swing2LimitDegrees=18, twistLimitDegrees=5}
 {bone1Name=dyn_belt_02_r, bone2Name=dyn_belt_01_r, swing1LimitDegrees=28, swing2LimitDegrees=30, twistLimitDegrees=5}
 {bone1Name=dyn_belt_03_r, bone2Name=dyn_belt_02_r, swing1LimitDegrees=26, swing2LimitDegrees=30, twistLimitDegrees=5}
 {bone1Name=dyn_belt_04_r, bone2Name=dyn_belt_03_r, swing1LimitDegrees=30, swing2LimitDegrees=30, twistLimitDegrees=5}
 {bone1Name=dyn_belt_05_r, bone2Name=dyn_belt_04_r, swing1LimitDegrees=30, swing2LimitDegrees=30, twistLimitDegrees=5}
 {bone1Name=dyn_skirt_bk_01_l, bone2Name=dyn_skirt_bk_01_l_MSTR, swing1LimitDegrees=5, swing2LimitDegrees=10, twistLimitDegrees=0}
 {bone1Name=dyn_skirt_bk_01_r, bone2Name=dyn_skirt_bk_01_r_MSTR, swing1LimitDegrees=5, swing2LimitDegrees=20, twistLimitDegrees=5}
 {
  bone1Name=dyn_skirt_fr_01_r_MSTR
  bone2Name=pelvis
  swing1LimitDegrees=45
  swing1Motion=Free
  swing2LimitDegrees=45
  swing2Motion=Free
  twistLimitDegrees=45
  twistMotion=Free
 }
 {
  bone1Name=calf_l
  bone2Name=thigh_l
  swing1LimitDegrees=45
  swing1Motion=Free
  swing2LimitDegrees=45
  swing2Motion=Free
  twistLimitDegrees=45
  twistMotion=Free
 }
 {
  bone1Name=calf_r
  bone2Name=thigh_r
  swing1LimitDegrees=45
  swing1Motion=Free
  swing2LimitDegrees=45
  swing2Motion=Free
  twistLimitDegrees=45
  twistMotion=Free
 }
 {
  bone1Name=dyn_collar_01_l_MSTR
  bone2Name=pelvis
  swing1LimitDegrees=45
  swing1Motion=Free
  swing2LimitDegrees=45
  swing2Motion=Free
  twistLimitDegrees=45
  twistMotion=Free
 }
 {
  bone1Name=dyn_collar_01_r_MSTR
  bone2Name=pelvis
  swing1LimitDegrees=45
  swing1Motion=Free
  swing2LimitDegrees=45
  swing2Motion=Free
  twistLimitDegrees=45
  twistMotion=Free
 }
 {
  bone1Name=thigh_l
  bone2Name=pelvis
  swing1LimitDegrees=45
  swing1Motion=Free
  swing2LimitDegrees=45
  swing2Motion=Free
  twistLimitDegrees=45
  twistMotion=Free
 }
 {
  bone1Name=thigh_r
  bone2Name=pelvis
  swing1LimitDegrees=45
  swing1Motion=Free
  swing2LimitDegrees=45
  swing2Motion=Free
  twistLimitDegrees=45
  twistMotion=Free
 }
 {
  bone1Name=dyn_skirt_bk_01_l_MSTR
  bone2Name=pelvis
  swing1LimitDegrees=45
  swing1Motion=Free
  swing2LimitDegrees=45
  swing2Motion=Free
  twistLimitDegrees=45
  twistMotion=Free
 }
 {
  bone1Name=dyn_skirt_bk_01_r_MSTR
  bone2Name=pelvis
  swing1LimitDegrees=45
  swing1Motion=Free
  swing2LimitDegrees=45
  swing2Motion=Free
  twistLimitDegrees=45
  twistMotion=Free
 }
 {
  bone1Name=dyn_belt_01_r_MSTR
  bone2Name=dyn_beltstrap_MSTR
  swing1LimitDegrees=45
  swing1Motion=Free
  swing2LimitDegrees=45
  swing2Motion=Free
  twistLimitDegrees=45
  twistMotion=Free
 }
 {
  bone1Name=dyn_skirt_01_l_MSTR
  bone2Name=pelvis
  swing1LimitDegrees=45
  swing1Motion=Free
  swing2LimitDegrees=45
  swing2Motion=Free
  twistLimitDegrees=45
  twistMotion=Free
 }
 {bone1Name=dyn_skirt_01_l, bone2Name=dyn_skirt_01_l_MSTR, swing1LimitDegrees=5, swing2LimitDegrees=30, twistLimitDegrees=15}
 {bone1Name=dyn_skirt_02_l, bone2Name=dyn_skirt_01_l, swing1LimitDegrees=5, swing2LimitDegrees=45, twistLimitDegrees=0}
 {
  bone1Name=dyn_skirt_01_r_MSTR
  bone2Name=pelvis
  swing1LimitDegrees=45
  swing1Motion=Free
  swing2LimitDegrees=45
  swing2Motion=Free
  twistLimitDegrees=45
  twistMotion=Free
 }
 {bone1Name=dyn_skirt_01_r, bone2Name=dyn_skirt_01_r_MSTR, swing1LimitDegrees=5, swing2LimitDegrees=30, twistLimitDegrees=5}
 {bone1Name=dyn_skirt_02_r, bone2Name=dyn_skirt_01_r, swing1LimitDegrees=5, swing2LimitDegrees=30, twistLimitDegrees=5}
 {bone1Name=dyn_skirt_fr_02_r, bone2Name=dyn_skirt_fr_01_r, swing1LimitDegrees=5, swing2LimitDegrees=5, twistLimitDegrees=5}
 {
  bone1Name=dyn_skirt_fr_01_l_MSTR
  bone2Name=pelvis
  swing1LimitDegrees=45
  swing1Motion=Free
  swing2LimitDegrees=45
  swing2Motion=Free
  twistLimitDegrees=45
  twistMotion=Free
 }
 {bone1Name=dyn_skirt_fr_01_l, bone2Name=dyn_skirt_fr_01_l_MSTR, swing1LimitDegrees=5, swing2LimitDegrees=30, twistLimitDegrees=4}
 {bone1Name=dyn_skirt_fr_02_l, bone2Name=dyn_skirt_fr_01_l, swing1LimitDegrees=5, swing2LimitDegrees=30, twistLimitDegrees=5}
 {
  bone1Name=dyn_beltstrap_MSTR
  bone2Name=pelvis
  swing1LimitDegrees=45
  swing1Motion=Free
  swing2LimitDegrees=45
  swing2Motion=Free
  twistLimitDegrees=45
  twistMotion=Free
 }
 {
  bone1Name=dyn_skirt_02_l
  bone2Name=dyn_skirt_fr_02_l
  swing1LimitDegrees=45
  swing1Motion=Free
  swing2LimitDegrees=45
  swing2Motion=Free
  twistLimitDegrees=45
  twistMotion=Free
 }
]
