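# Physics asset for the left boot-string dynamic bones of the female medium Black Widow Jacket player cosmetic.
# Capsule bodies cover the master string bone, inner and outer loops and ends, and the left foot.
# All four string pieces constrain directly to the master in a star layout, with identical swing limits and tighter twist.
# The large foot capsule has no constraint and supplies a collision shape for the boot.
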PhysicsAsset /BRCosmetics/Characters/Player/Female/Medium/Bodies/F_MED_Black_Widow_Jacket/Meshes/F_MED_Black_Widow_Jacket_Dyn_strings_l
bodies {
 dyn_boot_string_end_in_l [{shapeType=Capsule, length=3, radius=0.5, center=<5.1974 0.402 1.1398>, rotation=<rot 69.2345 -172.5486 -177.7775>}]
 dyn_boot_string_end_out_l [{shapeType=Capsule, length=2.875, radius=0.5, center=<5.1809 -0.511 0.6149>, rotation=<rot 79.6928 -170.2956 -160.5963>}]
 dyn_boot_string_l_MSTR [{shapeType=Capsule, length=0.5, radius=0.5}]
 dyn_boot_string_loop_out_l [{shapeType=Capsule, length=2.3125, radius=1.9062, center=<3.0261 0.187 0.9551>, rotation=<rot 47.1261 -114.6086 -110.5878>}]
 dyn_boot_string_roop_in_l [{shapeType=Capsule, length=2.1562, radius=2.0625, center=<3.072 -0.4548 0.307>, rotation=<rot -2.5968 -88.5646 -97.6585>}]
 foot_l [{shapeType=Capsule, length=12.5435, radius=6.1379, center=<-4.933 6.1805 0.0001>, rotation=<rot 90 0 60.0009>}]
}
constraints [
 {bone1Name=dyn_boot_string_end_out_l, bone2Name=dyn_boot_string_l_MSTR, swing1LimitDegrees=30, swing2LimitDegrees=30, twistLimitDegrees=5}
 {bone1Name=dyn_boot_string_loop_out_l, bone2Name=dyn_boot_string_l_MSTR, swing1LimitDegrees=30, swing2LimitDegrees=30, twistLimitDegrees=5}
 {bone1Name=dyn_boot_string_end_in_l, bone2Name=dyn_boot_string_l_MSTR, swing1LimitDegrees=30, swing2LimitDegrees=30, twistLimitDegrees=5}
 {bone1Name=dyn_boot_string_roop_in_l, bone2Name=dyn_boot_string_l_MSTR, swing1LimitDegrees=30, swing2LimitDegrees=30, twistLimitDegrees=5}
]
