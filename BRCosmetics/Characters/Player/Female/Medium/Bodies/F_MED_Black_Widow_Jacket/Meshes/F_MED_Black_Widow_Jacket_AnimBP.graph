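# Post-process style cosmetic AnimBP for the female Black Widow jacket body, copying the parent mesh pose and adding secondary motion.
# Weighted constraints drive the collar; additive bone offsets keyed to thigh angles and crouch adjust skirt and belt.
# Pelvis-space RigidBody simulates the skirt, with gravity switched for swim-sprinting and skydiving, then AnimDynamics handles thigh correctives.
# Choker and string RigidBody nodes are disconnected from the output.
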AnimBlueprint /BRCosmetics/Characters/Player/Female/Medium/Bodies/F_MED_Black_Widow_Jacket/Meshes/F_MED_Black_Widow_Jacket_AnimBP
target_skeleton=&/F_MED_Black_Widow_Jacket_Skeleton
graphs [
 AnimGraph {nodes [
   n1 = Get bIsCrouching()
   "Transform (Modify) Bone - Bone: dyn_skirt_fr_01_r_MSTR"!(ComponentPose, bAlphaBoolEnabled:bool=@n1, Translation="0,0,2", Rotation="0,-5,0", Scale="1,1,1") -> (Pose) {
    alphaInputType=Bool
    rotationMode=BMM_Additive
    rotationSpace=BCS_BoneSpace
    translationMode=BMM_Additive
    translationSpace=BCS_BoneSpace
    alphaBoolBlend {blendOption=Linear}
    alphaScaleBiasClamp {bInterpResult=true, interpSpeedIncreasing=40}
    boneToModify {boneName=dyn_skirt_fr_01_r_MSTR}
    rotation {yaw=-5}
    translation {z=2}
   }
   RigidBody(Alpha=0) {
    bClampLinearTranslationLimitToRefPose=true
    bOverrideWorldGravity=true
    overridePhysicsAsset=&/F_MED_Black_Widow_Jacket_Dyn_choker
    simulationSpace=BaseBoneSpace
    alphaBoolBlend {blendOption=Linear}
    baseBoneRef {boneName=neck_02}
    componentAppliedLinearAccClamp=<0 0 0>
   }
   n2 = RigidBody!(ComponentPose, Alpha, OverrideWorldGravity:Vector="0,0,-980", ExternalForce) -> (Pose) {
    bClampLinearTranslationLimitToRefPose=true
    bOverrideWorldGravity=true
    overridePhysicsAsset=&/F_MED_Black_Widow_Jacket_Dyn_strings_l
    simulationSpace=BaseBoneSpace
    alphaBoolBlend {blendOption=Linear}
    baseBoneRef {boneName=thigh_l}
    componentAppliedLinearAccClamp=<3000 3000 3000>
    componentLinearAccScale=<0.4 0.35 0.35>
    componentLinearVelScale=<0.4 0.35 0.35>
    overrideWorldGravity {z=-980}
    simSpaceSettings {maxAngularAcceleration=500, maxAngularVelocity=500, velocityScaleZ=0, worldAlpha=0.05, externalLinearDragV=<0.1 0.1 0.1>}
   }
   RigidBody(ComponentPose=@n2, OverrideWorldGravity="0,0,-980") {
    bClampLinearTranslationLimitToRefPose=true
    bOverrideWorldGravity=true
    overridePhysicsAsset=&/F_MED_Black_Widow_Jacket_Dyn_strings_r
    simulationSpace=BaseBoneSpace
    alphaBoolBlend {blendOption=Linear}
    baseBoneRef {boneName=thigh_r}
    componentAppliedLinearAccClamp=<3000 3000 3000>
    componentLinearAccScale=<0.4 0.35 0.35>
    componentLinearVelScale=<0.4 0.35 0.35>
    overrideWorldGravity {z=-980}
    simSpaceSettings {maxAngularAcceleration=500, maxAngularVelocity=500, velocityScaleZ=0, worldAlpha=0.05, externalLinearDragV=<0.1 0.1 0.1>}
   }
   n3 = Get thigh_rotation_l() -> (thigh_rotation_l:double)
   n4 = "Constraint - dyn_collar_fr_surface_top (PPPPPPP)"!(
    ComponentPose=Local To Component(LocalPose=Copy Pose From Mesh() {bUseAttachedParent=true})
    Alpha=1
    ConstraintWeights_0=0.0822
    ConstraintWeights_1:float=0.0122
    ConstraintWeights_2:float=0.3637
    ConstraintWeights_3:float=0.236
    ConstraintWeights_4:float=0.0048
    ConstraintWeights_5:float=0.1266
    ConstraintWeights_6:float=0.1746
   ) -> (Pose) {
    alphaBoolBlend {blendOption=Linear}
    boneToModify {boneName=dyn_collar_fr_surface_top}
    constraintSetup [
     {targetBone {boneName=head}}
     {targetBone {boneName=clavicle_l}}
     {targetBone {boneName=spine_05}}
     {targetBone {boneName=neck_02}}
     {targetBone {boneName=clavicle_r}}
     {targetBone {boneName=C_jaw}}
     {targetBone {boneName=neck_01}}
    ]
    constraintWeights [1,1,1,1,1,1,1]
   }
   n5 = "Constraint - dyn_collar_fr_surface_mid (PPPPPPP)"!(
    ComponentPose=@n4
    Alpha=1
    ConstraintWeights_0=0.0416
    ConstraintWeights_1:float=0.0292
    ConstraintWeights_2:float=0.5883
    ConstraintWeights_3:float=0.146
    ConstraintWeights_4:float=0.0075
    ConstraintWeights_5:float=0.0601
    ConstraintWeights_6:float=0.1272
   ) -> (Pose) {
    alphaBoolBlend {blendOption=Linear}
    boneToModify {boneName=dyn_collar_fr_surface_mid}
    constraintSetup [
     {targetBone {boneName=head}}
     {targetBone {boneName=clavicle_l}}
     {targetBone {boneName=spine_05}}
     {targetBone {boneName=neck_02}}
     {targetBone {boneName=clavicle_r}}
     {targetBone {boneName=C_jaw}}
     {targetBone {boneName=neck_01}}
    ]
    constraintWeights [1,1,1,1,1,1,1]
   }
   n6 = "Constraint - dyn_collar_fr_surface_bott (PPPPPPP)"!(
    ComponentPose=@n5
    Alpha=1
    ConstraintWeights_0=0.0416
    ConstraintWeights_1:float=0.0292
    ConstraintWeights_2:float=0.5883
    ConstraintWeights_3:float=0.146
    ConstraintWeights_4:float=0.0075
    ConstraintWeights_5:float=0.0601
    ConstraintWeights_6:float=0.1272
   ) -> (Pose) {
    alphaBoolBlend {blendOption=Linear}
    boneToModify {boneName=dyn_collar_fr_surface_bott}
    constraintSetup [
     {targetBone {boneName=head}}
     {targetBone {boneName=clavicle_l}}
     {targetBone {boneName=spine_05}}
     {targetBone {boneName=neck_02}}
     {targetBone {boneName=clavicle_r}}
     {targetBone {boneName=C_jaw}}
     {targetBone {boneName=neck_01}}
    ]
    constraintWeights [1,1,1,1,1,1,1]
   }
   n7 = "Transform (Modify) Bone - Bone: dyn_skirt_fr_01_l_MSTR"(ComponentPose=@n6, Alpha=@n3, Translation="-5.516,-1.854,-1.158", Rotation="74.982,-65.141,-58.431", Scale="1,1,1") {
    rotationMode=BMM_Additive
    rotationSpace=BCS_BoneSpace
    translationMode=BMM_Additive
    translationSpace=BCS_BoneSpace
    alphaBoolBlend {blendOption=Linear}
    alphaScaleBiasClamp {interpSpeedIncreasing=40}
    boneToModify {boneName=dyn_skirt_fr_01_l_MSTR}
    rotation=<rot 74.9818 -65.1411 -58.4308>
    translation=<-5.5164 -1.854 -1.1579>
   }
   n8 = "Transform (Modify) Bone - Bone: dyn_skirt_01_l_MSTR"(ComponentPose=@n7, Alpha=@n3, Translation="-2.578,-2.166,-0.000", Rotation="0,0,0", Scale="1,1,1") {
    rotationMode=BMM_Additive
    rotationSpace=BCS_BoneSpace
    translationMode=BMM_Additive
    translationSpace=BCS_BoneSpace
    alphaBoolBlend {blendOption=Linear}
    alphaScaleBiasClamp {interpSpeedIncreasing=40}
    boneToModify {boneName=dyn_skirt_01_l_MSTR}
    translation=<-2.578 -2.166>
   }
   n9 = Get thigh rotation_r() -> (thigh rotation_r:double)
   n10 = "Transform (Modify) Bone - Bone: dyn_beltstrap_MSTR"(ComponentPose=@n8, Alpha=@n9, Translation="-5.693,0.890,1.021", Rotation="48.972,11.692,15.337", Scale="1,1,1") {
    rotationMode=BMM_Additive
    rotationSpace=BCS_BoneSpace
    translationMode=BMM_Additive
    translationSpace=BCS_BoneSpace
    alphaBoolBlend {blendOption=Linear}
    alphaScaleBiasClamp {bInterpResult=true, interpSpeedDecreasing=40, interpSpeedIncreasing=40}
    boneToModify {boneName=dyn_beltstrap_MSTR}
    rotation=<rot 48.9719 11.6924 15.3374>
    translation=<-5.693 0.89 1.021>
   }
   n11 = "Transform (Modify) Bone - Bone: dyn_skirt_01_r_MSTR"(ComponentPose=@n10, Alpha=@n9, Translation="-6.726,-0.288,0.120", Rotation="-10.000,-15.001,0.000", Scale="1,1,1") {
    rotationMode=BMM_Additive
    rotationSpace=BCS_BoneSpace
    translationMode=BMM_Additive
    translationSpace=BCS_BoneSpace
    alphaBoolBlend {blendOption=Linear}
    alphaScaleBiasClamp {interpSpeedIncreasing=40}
    boneToModify {boneName=dyn_skirt_01_r_MSTR}
    rotation {pitch=-10, yaw=-15.0014}
    translation=<-6.726 -0.288 0.12>
   }
   n12 = "Constraint - dyn_skirt_fr_01_r_MSTR (PP)"!(ComponentPose=@n11, Alpha=1, ConstraintWeights_0=1, ConstraintWeights_1:float=0) -> (Pose) {
    alphaBoolBlend {blendOption=Linear}
    boneToModify {boneName=dyn_skirt_fr_01_r_MSTR}
    constraintSetup [{targetBone {boneName=thigh_twist_01_r}}, {targetBone {boneName=pelvis}}]
    constraintWeights [1,1]
   }
   n13 = "Transform (Modify) Bone - Bone: dyn_skirt_fr_01_l_MSTR"(
    ComponentPose=@n12
    Alpha=float * float(A=Get thigh_rotation_l_side() -> (thigh_rotation_l_side:double)) {is=title K2Node_PromotableOperator}
    Translation="0,-7,0"
    Rotation="-38.016,-1.413,-26.065"
    Scale="1,1,1"
   ) {
    rotationMode=BMM_Additive
    rotationSpace=BCS_BoneSpace
    translationMode=BMM_Additive
    translationSpace=BCS_BoneSpace
    alphaBoolBlend {blendOption=Linear}
    alphaScaleBiasClamp {bInterpResult=true, interpSpeedIncreasing=40}
    boneToModify {boneName=dyn_skirt_fr_01_l_MSTR}
    rotation=<rot -38.016 -1.413 -26.065>
    translation {y=-7}
   }
   n14 = float * float(A=Get thigh_rotation_r_side() -> (thigh_rotation_r_side:double))
   n15 = "Transform (Modify) Bone - Bone: dyn_skirt_fr_01_r_MSTR"!(ComponentPose=@n13, Alpha:float=@n14, Translation="0,2,4", Rotation="5,0,0", Scale="1,1,1") -> (Pose) {
    rotationMode=BMM_Additive
    rotationSpace=BCS_BoneSpace
    translationMode=BMM_Additive
    translationSpace=BCS_BoneSpace
    alphaBoolBlend {blendOption=Linear}
    alphaScaleBiasClamp {interpSpeedIncreasing=40}
    boneToModify {boneName=dyn_skirt_fr_01_r_MSTR}
    rotation {pitch=5}
    translation {y=2, z=4}
   }
   n16 = "Transform (Modify) Bone - Bone: dyn_skirt_01_r_MSTR"(ComponentPose=@n15, Alpha=@n14, Translation="0,0,3", Rotation="10,0,0", Scale="1,1,1") {
    rotationMode=BMM_Additive
    rotationSpace=BCS_BoneSpace
    translationMode=BMM_Additive
    translationSpace=BCS_BoneSpace
    alphaBoolBlend {blendOption=Linear}
    alphaScaleBiasClamp {interpSpeedIncreasing=40}
    boneToModify {boneName=dyn_skirt_01_r_MSTR}
    rotation {pitch=10}
    translation {z=3}
   }
   n17 = RigidBody(ComponentPose=@n16, OverrideWorldGravity=Get gravity_override() -> (gravity_override:Vector)) {
    bClampLinearTranslationLimitToRefPose=true
    bOverrideWorldGravity=true
    overridePhysicsAsset=&/F_MED_Black_Widow_Jacket_Dyn_skirt
    simulationSpace=BaseBoneSpace
    alphaBoolBlend {blendOption=Linear}
    baseBoneRef {boneName=pelvis}
    componentAppliedLinearAccClamp=<3000 3000 3000>
    componentLinearAccScale {y=0.35, z=0.35}
    componentLinearVelScale {y=0.35, z=0.35}
    overrideWorldGravity {z=-980}
    simSpaceSettings {maxAngularAcceleration=500, maxAngularVelocity=500, velocityScaleZ=0, worldAlpha=0.05, externalLinearDragV=<0.1 0.1 0.1>}
   }
   n18 = "Transform (Modify) Bone - Bone: dyn_skirt_02_l"!(ComponentPose=@n17, bAlphaBoolEnabled:bool=@n1, Translation="0,3,4", Rotation="0,0,0", Scale="1,1,1") -> (Pose) {
    alphaInputType=Bool
    rotationMode=BMM_Additive
    rotationSpace=BCS_BoneSpace
    translationMode=BMM_Additive
    translationSpace=BCS_BoneSpace
    alphaBoolBlend {blendOption=Linear}
    alphaScaleBiasClamp {bInterpResult=true, interpSpeedIncreasing=40}
    boneToModify {boneName=dyn_skirt_02_l}
    translation {y=3, z=4}
   }
   n19 = "Transform (Modify) Bone - Bone: dyn_skirt_fr_01_l"!(ComponentPose=@n18, bAlphaBoolEnabled:bool=@n1, Translation="0,0,0", Rotation="0,0,0", Scale="1,1,1") -> (Pose) {
    alphaInputType=Bool
    rotationMode=BMM_Additive
    rotationSpace=BCS_BoneSpace
    translationMode=BMM_Additive
    translationSpace=BCS_BoneSpace
    alphaBoolBlend {blendOption=Linear}
    alphaScaleBiasClamp {bInterpResult=true, interpSpeedIncreasing=40}
    boneToModify {boneName=dyn_skirt_fr_01_l}
   }
   n20 = "Transform (Modify) Bone - Bone: dyn_skirt_fr_02_l"!(ComponentPose=@n19, bAlphaBoolEnabled:bool=@n1, Translation="0,0,0", Rotation="0,0,0", Scale="1,1,1") -> (Pose) {
    alphaInputType=Bool
    rotationMode=BMM_Additive
    rotationSpace=BCS_BoneSpace
    translationMode=BMM_Additive
    translationSpace=BCS_BoneSpace
    alphaBoolBlend {blendOption=Linear}
    alphaScaleBiasClamp {bInterpResult=true, interpSpeedIncreasing=40}
    boneToModify {boneName=dyn_skirt_fr_02_l}
   }
   n21 = NOT Boolean(A=@n1)
   n22 = "Transform (Modify) Bone - Bone: dyn_skirt_fr_01_r_MSTR"!(ComponentPose=@n20, bAlphaBoolEnabled:bool=@n21, Translation="0,2,2.5", Rotation="-10,0,0", Scale="1,1,1") -> (Pose) {
    alphaInputType=Bool
    rotationMode=BMM_Additive
    rotationSpace=BCS_BoneSpace
    translationMode=BMM_Additive
    translationSpace=BCS_BoneSpace
    alphaBoolBlend {blendOption=Linear}
    alphaScaleBiasClamp {bInterpResult=true, interpSpeedIncreasing=40}
    boneToModify {boneName=dyn_skirt_fr_01_r_MSTR}
    rotation {pitch=-10}
    translation {y=2, z=2.5}
   }
   n23 = "Transform (Modify) Bone - Bone: dyn_skirt_02_l"(ComponentPose=@n22, bAlphaBoolEnabled=@n21, Translation="0,0,1.5", Rotation="0,0,0", Scale="1,1,1") {
    alphaInputType=Bool
    rotationMode=BMM_Additive
    rotationSpace=BCS_BoneSpace
    translationMode=BMM_Additive
    translationSpace=BCS_BoneSpace
    alphaBoolBlend {blendOption=Linear}
    alphaScaleBiasClamp {bInterpResult=true, interpSpeedIncreasing=40}
    boneToModify {boneName=dyn_skirt_02_l}
    translation {z=1.5}
   }
   n24 = "Transform (Modify) Bone - Bone: ALT_corrective_L"!(ComponentPose=@n23, bAlphaBoolEnabled:bool=True, Translation="5,0,0", Rotation="4,0,0", Scale="1,1,1") -> (Pose) {
    alphaInputType=Bool
    rotationMode=BMM_Additive
    rotationSpace=BCS_BoneSpace
    translationMode=BMM_Additive
    translationSpace=BCS_BoneSpace
    alphaBoolBlend {blendOption=Linear}
    alphaScaleBiasClamp {bInterpResult=true, interpSpeedIncreasing=40}
    boneToModify {boneName=ALT_corrective_L}
    rotation {pitch=4}
    translation {x=5}
   }
   n25 = "Transform (Modify) Bone - Bone: ALT_corrective_R"!(ComponentPose=@n24, bAlphaBoolEnabled:bool=True, Translation="-6,0,0", Rotation="4,0,0", Scale="1,1,1") -> (Pose) {
    alphaInputType=Bool
    rotationMode=BMM_Additive
    rotationSpace=BCS_BoneSpace
    translationMode=BMM_Additive
    translationSpace=BCS_BoneSpace
    alphaBoolBlend {blendOption=Linear}
    alphaScaleBiasClamp {bInterpResult=true, interpSpeedIncreasing=40}
    boneToModify {boneName=ALT_corrective_R}
    rotation {pitch=4}
    translation {x=-6}
   }
   n26 = "Transform (Modify) Bone - Bone: ALT_corrective_Back"!(ComponentPose=@n25, bAlphaBoolEnabled:bool=True, Translation="5,0,0", Rotation="0,0,0", Scale="1,1,1") -> (Pose) {
    alphaInputType=Bool
    rotationMode=BMM_Additive
    rotationSpace=BCS_BoneSpace
    translationMode=BMM_Additive
    translationSpace=BCS_BoneSpace
    alphaBoolBlend {blendOption=Linear}
    alphaScaleBiasClamp {bInterpResult=true, interpSpeedIncreasing=40}
    boneToModify {boneName=ALT_corrective_Back}
    translation {x=5}
   }
   n27 = Anim Dynamics - Bone: ALT_corrective_L!(ComponentPose=@n26, bAlphaBoolEnabled:bool=True, ExternalForce="0,0,0") -> (Pose) {
    alphaInputType=Bool
    angularSpringConstant=75
    bAngularSpring=true
    bUseGravityOverride=true
    bUsePlanarLimit=false
    bUseSphericalLimits=true
    simulationSpace=BoneRelative
    alphaBoolBlend {blendOption=Linear}
    boundBone {boneName=ALT_corrective_L}
    gravityOverride {z=-980}
    physicsBodyDefinitions [{
      boundBone {boneName=ALT_corrective_L}
      boxExtents=<10 10 10>
      constraintSetup {angularLimitsMax=<0 0 0>, angularLimitsMin=<0 -15 0>, angularTarget=<0 0 0>}
      localJointOffset=<0 0 0>
     }]
    relativeSpaceBone {boneName=thigh_l}
    sphericalLimits [{limitRadius=9, drivingBone {boneName=thigh_l}, sphereLocalOffset=<-5 0 0>}]
   }
   n28 = Anim Dynamics - Bone: ALT_corrective_R!(ComponentPose=@n27, bAlphaBoolEnabled:bool=True, ExternalForce="0,0,0") -> (Pose) {
    alphaInputType=Bool
    angularSpringConstant=75
    bAngularSpring=true
    bUseGravityOverride=true
    bUsePlanarLimit=false
    bUseSphericalLimits=true
    simulationSpace=BoneRelative
    alphaBoolBlend {blendOption=Linear}
    boundBone {boneName=ALT_corrective_R}
    gravityOverride {z=-980}
    physicsBodyDefinitions [{
      boundBone {boneName=ALT_corrective_R}
      boxExtents=<10 10 10>
      constraintSetup {angularLimitsMax=<0 0 0>, angularLimitsMin=<0 -15 0>, angularTarget=<0 0 0>}
      localJointOffset=<0 0 0>
     }]
    relativeSpaceBone {boneName=thigh_r}
    sphericalLimits [{limitRadius=9, drivingBone {boneName=thigh_r}, sphereLocalOffset=<5 0 0>}]
   }
   n29 = "Transform (Modify) Bone - Bone: thigh_l"!(ComponentPose=@n28, bAlphaBoolEnabled:bool=True, Translation="0,0,0", Rotation="0,0,0", Scale="0.9,0.9,0.9") -> (Pose) {
    alphaInputType=Bool
    rotationMode=BMM_Additive
    rotationSpace=BCS_BoneSpace
    translationMode=BMM_Additive
    translationSpace=BCS_BoneSpace
    alphaBoolBlend {blendOption=Linear}
    alphaScaleBiasClamp {bInterpResult=true, interpSpeedIncreasing=40}
    boneToModify {boneName=thigh_l}
    scale=<0.9 0.9 0.9>
   }
   n30 = "Transform (Modify) Bone - Bone: thigh_r"!(ComponentPose=@n29, bAlphaBoolEnabled:bool=True, Translation="0,0,0", Rotation="0,0,0", Scale="0.9,0.9,0.9") -> (Pose) {
    alphaInputType=Bool
    rotationMode=BMM_Additive
    rotationSpace=BCS_BoneSpace
    translationMode=BMM_Additive
    translationSpace=BCS_BoneSpace
    alphaBoolBlend {blendOption=Linear}
    alphaScaleBiasClamp {bInterpResult=true, interpSpeedIncreasing=40}
    boneToModify {boneName=thigh_r}
    scale=<0.9 0.9 0.9>
   }
   n31 = "Transform (Modify) Bone - Bone: dyn_belt_01_r_MSTR"!(ComponentPose=@n30, bAlphaBoolEnabled:bool=True, Translation="0,0,0", Rotation="0,35,0", Scale="1,1,1") -> (Pose) {
    alphaInputType=Bool
    rotationMode=BMM_Additive
    rotationSpace=BCS_BoneSpace
    translationMode=BMM_Additive
    translationSpace=BCS_BoneSpace
    alphaBoolBlend {blendOption=Linear}
    alphaScaleBiasClamp {bInterpResult=true, interpSpeedIncreasing=40}
    boneToModify {boneName=dyn_belt_01_r_MSTR}
    rotation {yaw=35}
   }
   Output Pose(Result=Component To Local(ComponentPose=@n31)) {layerGroup=None}
  ]}
 EventGraph {nodes [
   Event BlueprintUpdateAnimation()
   > n1 = Sequence!(execute) -> (then_0, then_1, then_2:Exec, then_3:Exec)
   > n2 = Branch(Condition=Get bIsSwimSprinting() -> (bIsSwimSprinting:bool))
   > Set gravity_override(gravity_override="0, 0, 0")
   n3 = Branch(execute=@n2.else, Condition=Get bIsSkyDiving())
   > n4 = Branch(Condition=Get bIsParachuteOpened())
   Set gravity_override(execute=[@n3.else, @n4.then], gravity_override="0, 0,-980")
   n5 = Branch(execute=@n4.else, Condition="float > float"(A=Get PawnVelocityZ(), B=50) {is=title K2Node_CallFunction})
   > Set gravity_override(gravity_override="0, 0,300")
   Set gravity_override(execute=@n5.else, gravity_override="0, 0,-200")
   Set thigh rotation_r(
    execute=@n1.then_1
    thigh rotation_r:double=GetAlphaFromReferenceTransformDelta(SkeletalMeshComponent=Get SkelMesh(), BoneName=thigh_r, BaseName=pelvis, InputRange=(X=-50,Y=-90), TwistAxis="1,0,0")
   ) -> (Output_Get:double)
   > Set thigh_rotation_l(thigh_rotation_l:double=GetAlphaFromReferenceTransformDelta(SkeletalMeshComponent=Get SkelMesh(), BoneName=thigh_l, BaseName=pelvis, InputRange=(X=-50,Y=-90), TwistAxis="1,0,0")) -> (Output_Get:double)
   > Set thigh_rotation_l_side(thigh_rotation_l_side:double=GetAlphaFromReferenceTransformDelta(SkeletalMeshComponent=Get SkelMesh(), BoneName=thigh_l, BaseName=pelvis, DeltaComponentType=EulerY, InputRange=(X=9,Y=17), TwistAxis="1,0,0")) -> (Output_Get:double)
   > Set thigh_rotation_r_side(thigh_rotation_r_side:double=GetAlphaFromReferenceTransformDelta(SkeletalMeshComponent=Get SkelMesh(), BoneName=thigh_r, BaseName=pelvis, DeltaComponentType=EulerY, InputRange=(X=7,Y=15), TwistAxis="1,0,0")) -> (Output_Get:double)
   Set Rigid_Body_Enable(execute=@n1.then_2, Rigid_Body_Enable:bool=NOT Boolean(A=Get bRigidBodyEnabled())) -> (Output_Get:bool)
   Set Full_fly(
    execute=@n1.then_3
    Full_fly:bool=OR Boolean!(A=Get bIsSkyDiving(), B=Get bIsSkydiveFloating(), C:bool=Get bIsSkydiveDiveMode()) -> (ReturnValue)
   ) -> (Output_Get:bool)
   Event BlueprintInitializeAnimation()
   > Set SkelMesh(SkelMesh=GetOwnerSkeletalMeshComponent())
  ]}
]
variables [
 Full_fly {type=boolean}
 gravity_override {type=Vector}
 Rigid_Body_Enable {type=boolean}
 SkelMesh {type=SkeletalMeshComponent}
 thigh rotation_r {type=number}
 thigh_rotation_l {type=number}
 thigh_rotation_l_side {type=number}
 thigh_rotation_r_side {type=number}
]
