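# Physics asset for the female medium Black Widow Jacket player cosmetic, set up for dynamic hair on the head.
# Five three-bone hair chains (back, back-left, back-right, left, right) hang from the head as small capsules, with tight twist and one swing axis, wider on the other.
# Larger capsules on the head, neck, upper spine and clavicles form the body volumes, linked by joints with free swing and twist.
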PhysicsAsset /BRCosmetics/Characters/Player/Female/Medium/Bodies/F_MED_Black_Widow_Jacket/Meshes/Parts/F_MED_Black_Widow_Jacket_Dyn
bodies {
 clavicle_l [{shapeType=Capsule, length=11.3047, radius=7.2035, center=<5.1748 -0.7569 -1.363>, rotation=<rot 87.73 0 -2.8936>}]
 clavicle_r [{shapeType=Capsule, length=10.6135, radius=7.8948, center=<-5.1748 0.7569 1.363>, rotation=<rot 87.73 0 -2.8936>}]
 dyn_hair_bk_01 [{shapeType=Capsule, length=2, radius=2, center=<2.2 0 0>, rotation=<rot 90 0 0>}]
 dyn_hair_bk_02 [{shapeType=Capsule, length=2, radius=2, center=<2.2 0 0>, rotation=<rot 90 0 0>}]
 dyn_hair_bk_03 [{shapeType=Capsule, length=0.0144, radius=2, center=<1.6362 0 0>, rotation=<rot 90 0 0>}]
 dyn_hair_bk_l_01 [{shapeType=Capsule, length=2.125, radius=1.875, center=<2.2 0 0>, rotation=<rot 90 0 0>}]
 dyn_hair_bk_l_02 [{shapeType=Capsule, length=2.125, radius=1.875, center=<2.2 0 0>, rotation=<rot 90 0 0>}]
 dyn_hair_bk_l_03 [{shapeType=Capsule, length=0.1394, radius=1.875, center=<1.6362 0 0>, rotation=<rot 90 0 0>}]
 dyn_hair_bk_r_01 [{shapeType=Capsule, length=2.1562, radius=1.8438, center=<2.2 0 0>, rotation=<rot 90 0 0>}]
 dyn_hair_bk_r_02 [{shapeType=Capsule, length=2.1562, radius=1.8438, center=<2.2 0 0>, rotation=<rot 90 0 0>}]
 dyn_hair_bk_r_03 [{shapeType=Capsule, length=0.1707, radius=1.8438, center=<1.6362 0 0>, rotation=<rot 90 0 0>}]
 dyn_hair_l_01 [{shapeType=Capsule, length=2.5312, radius=1.4688, center=<2.2 0 0>, rotation=<rot 90 0 0>}]
 dyn_hair_l_02 [{shapeType=Capsule, length=2.6875, radius=1.3125, center=<2.2 0 0>, rotation=<rot 90 0 0>}]
 dyn_hair_l_03 [{shapeType=Capsule, length=0.7019, radius=1.3125, center=<1.6362 0 0>, rotation=<rot 90 0 0>}]
 dyn_hair_r_01 [{shapeType=Capsule, length=2.5312, radius=1.4688, center=<2.2 0 0>, rotation=<rot 90 0 0>}]
 dyn_hair_r_02 [{shapeType=Capsule, length=2.6875, radius=1.3125, center=<2.2 0 0>, rotation=<rot 90 0 0>}]
 dyn_hair_r_03 [{shapeType=Capsule, length=0.7019, radius=1.3125, center=<1.6362 0 0>, rotation=<rot 90 0 0>}]
 head [
  {shapeType=Capsule, length=4.0846, radius=5.9154, center=<-0.0738 -0.3866 0.0001>, rotation=<rot 90 0 -10>}
  {shapeType=Capsule, length=6.1326, radius=5.8777, center=<2.1232 3.0881 0>, rotation=<rot 90 -90 -95>}
  {shapeType=Capsule, length=1.2892, radius=1.7232, center=<2.1169 -0.3915 6.1713>, rotation=<rot 85.9796 -174.9995 -180>}
  {shapeType=Capsule, length=1.2892, radius=1.7232, center=<2.0062 -0.5593 -6.3643>, rotation=<rot 81.8929 5.0021 0>}
 ]
 neck_01 [{shapeType=Capsule, length=5.9622, radius=6.4996, center=<1.1864 0.9611 0.0001>, rotation=<rot 90 0 0>}]
 spine_05 [{shapeType=Capsule, length=7.8353, radius=8.5564, center=<4.8478 -0.9019 0>}]
}
constraints [
 {bone1Name=dyn_hair_bk_l_01, bone2Name=head, swing1LimitDegrees=10, swing2LimitDegrees=45, twistLimitDegrees=5}
 {bone1Name=dyn_hair_bk_l_02, bone2Name=dyn_hair_bk_l_01, swing1LimitDegrees=10, swing2LimitDegrees=45, twistLimitDegrees=5}
 {bone1Name=dyn_hair_bk_l_03, bone2Name=dyn_hair_bk_l_02, swing1LimitDegrees=10, swing2LimitDegrees=45, twistLimitDegrees=5}
 {bone1Name=dyn_hair_bk_01, bone2Name=head, swing1LimitDegrees=10, swing2LimitDegrees=45, twistLimitDegrees=5}
 {bone1Name=dyn_hair_bk_02, bone2Name=dyn_hair_bk_01, swing1LimitDegrees=10, swing2LimitDegrees=45, twistLimitDegrees=5}
 {bone1Name=dyn_hair_bk_03, bone2Name=dyn_hair_bk_02, swing1LimitDegrees=10, swing2LimitDegrees=45, twistLimitDegrees=5}
 {bone1Name=dyn_hair_r_01, bone2Name=head, swing1LimitDegrees=10, swing2LimitDegrees=45, twistLimitDegrees=5}
 {bone1Name=dyn_hair_r_02, bone2Name=dyn_hair_r_01, swing1LimitDegrees=10, swing2LimitDegrees=45, twistLimitDegrees=5}
 {bone1Name=dyn_hair_r_03, bone2Name=dyn_hair_r_02, swing1LimitDegrees=10, swing2LimitDegrees=45, twistLimitDegrees=5}
 {bone1Name=dyn_hair_bk_r_01, bone2Name=head, swing1LimitDegrees=10, swing2LimitDegrees=45, twistLimitDegrees=5}
 {bone1Name=dyn_hair_bk_r_02, bone2Name=dyn_hair_bk_r_01, swing1LimitDegrees=10, swing2LimitDegrees=45, twistLimitDegrees=5}
 {bone1Name=dyn_hair_bk_r_03, bone2Name=dyn_hair_bk_r_02, swing1LimitDegrees=10, swing2LimitDegrees=45, twistLimitDegrees=5}
 {
  bone1Name=head
  bone2Name=neck_01
  swing1LimitDegrees=45
  swing1Motion=Free
  swing2LimitDegrees=45
  swing2Motion=Free
  twistLimitDegrees=45
  twistMotion=Free
 }
 {
  bone1Name=clavicle_r
  bone2Name=spine_05
  swing1LimitDegrees=45
  swing1Motion=Free
  swing2LimitDegrees=45
  swing2Motion=Free
  twistLimitDegrees=45
  twistMotion=Free
 }
 {
  bone1Name=clavicle_l
  bone2Name=spine_05
  swing1LimitDegrees=45
  swing1Motion=Free
  swing2LimitDegrees=45
  swing2Motion=Free
  twistLimitDegrees=45
  twistMotion=Free
 }
 {
  bone1Name=neck_01
  bone2Name=spine_05
  swing1LimitDegrees=45
  swing1Motion=Free
  swing2LimitDegrees=45
  swing2Motion=Free
  twistLimitDegrees=45
  twistMotion=Free
 }
 {bone1Name=dyn_hair_l_01, bone2Name=head, swing1LimitDegrees=10, swing2LimitDegrees=45, twistLimitDegrees=5}
 {bone1Name=dyn_hair_l_02, bone2Name=dyn_hair_l_01, swing1LimitDegrees=10, swing2LimitDegrees=45, twistLimitDegrees=5}
 {bone1Name=dyn_hair_l_03, bone2Name=dyn_hair_l_02, swing1LimitDegrees=10, swing2LimitDegrees=45, twistLimitDegrees=5}
]
