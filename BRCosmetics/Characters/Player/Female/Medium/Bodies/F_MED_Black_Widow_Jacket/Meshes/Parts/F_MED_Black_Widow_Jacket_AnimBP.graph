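# Post-process-style cosmetic AnimBP for the female medium Black Widow jacket body part, adding physics-driven secondary motion.
# It copies the attached parent mesh's pose, then runs a RigidBody node with the jacket physics asset, simulated relative to neck_01.
# Each update sets gravity override by movement state: zero when swim-sprinting, velocity-dependent lift or reduced pull in closed-parachute skydiving, normal otherwise.
# A neck_01 debug-axis draw is left unconnected.
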AnimBlueprint /BRCosmetics/Characters/Player/Female/Medium/Bodies/F_MED_Black_Widow_Jacket/Meshes/Parts/F_MED_Black_Widow_Jacket_AnimBP
target_skeleton=&/F_MED_Black_Widow_Jacket_Skeleton
graphs [
 AnimGraph {nodes [
   n1 = RigidBody!(
    ComponentPose=Local To Component(LocalPose=Copy Pose From Mesh() {bUseAttachedParent=true})
    Alpha
    OverrideWorldGravity:Vector=Get GravityOverride()
    ExternalForce
   ) -> (Pose) {
    bClampLinearTranslationLimitToRefPose=true
    bOverrideWorldGravity=true
    overridePhysicsAsset=&/F_MED_Black_Widow_Jacket_Dyn
    simulationSpace=BaseBoneSpace
    alphaBoolBlend {blendOption=Linear}
    baseBoneRef {boneName=neck_01}
    componentAppliedLinearAccClamp=<3000 3000 3000>
    componentLinearAccScale=<0.2 0.1>
    componentLinearVelScale=<0.2 0.1>
    overrideWorldGravity {z=-980}
    simSpaceSettings {maxAngularAcceleration=5, maxAngularVelocity=5, maxLinearAcceleration=100, maxLinearVelocity=100, worldAlpha=1, externalLinearDragV=<1 1 1>}
   }
   Output Pose(Result=Component To Local(ComponentPose=@n1)) {layerGroup=None}
  ]}
 EventGraph {nodes [
   Event BlueprintUpdateAnimation()
   > n1 = Sequence()
   > n2 = Branch(Condition=Get bIsSwimSprinting() -> (bIsSwimSprinting:bool))
   > Set GravityOverride(GravityOverride="0, 0, 0")
   n3 = Branch(execute=@n2.else, Condition=Get bIsSkyDiving())
   > n4 = Branch(Condition=Get bIsParachuteOpened())
   Set GravityOverride(execute=[@n3.else, @n4.then], GravityOverride="0, 0,-980")
   n5 = Branch(execute=@n4.else, Condition="float > float"(A=Get PawnVelocityZ(), B=50) {is=title K2Node_CallFunction})
   > Set GravityOverride(GravityOverride="0, 0,300")
   Set GravityOverride(execute=@n5.else, GravityOverride="0, 0,-200")
   Reroute Node(InputPin=@n1.then_1)
   Event BlueprintInitializeAnimation()
   > Set Skel_Mesh(Skel_Mesh=GetOwnerSkeletalMeshComponent())
   n6 = Get Skel_Mesh()
   DrawDebugCoordinateSystem!(
    execute
    AxisLoc:Vector=vector + vector(
     A:Vector=GetSocketLocation!(self:Scene Component Object Reference=@n6, InSocketName:Name=neck_01) -> (ReturnValue:Vector)
     B:Vector="0, 0, 0"
    ) -> (ReturnValue:Vector) {is=title K2Node_CommutativeAssociativeBinaryOperator}
    AxisRot:Rotator=GetSocketRotation!(self:Scene Component Object Reference=@n6, InSocketName:Name=neck_01) -> (ReturnValue:Rotator)
    Scale:float=100
    Duration:float=0
    Thickness:float=1
    DepthPriority:EDrawDebugSceneDepthPriorityGroup Enum=World
   ) -> (then)
  ]}
]
variables [
 GravityOverride {type=Vector}
 Skel_Mesh {type=SkeletalMeshComponent}
]
